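# Sound cue for the BerryCard dial tone, playing a randomly chosen phone dialing sequence.
# A Random node picks equally among four Concatenators, each chaining the same ten Phone_Prank_Dial waves in a different order, with a trailing empty input.
# The cue slightly raises pitch, boosts volume and overrides attenuation with its own falloff distance.
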SoundCue /WildEstate/Gameplay/BerryCard/Audio/SC_BerryCard_DialTone_Cue
root=Random_1
nodes [
 {
  id=Concatenator_1
  children [
   WavePlayer_55
   WavePlayer_54
   WavePlayer_53
   WavePlayer_52
   WavePlayer_56
   WavePlayer_61
   WavePlayer_60
   WavePlayer_59
   WavePlayer_58
   WavePlayer_57
   null
  ]
  properties {inputVolume [1,1,1,1,1,1,1,1,1,1,1]}
 }
 {
  id=Concatenator_2
  children [
   WavePlayer_7
   WavePlayer_24
   WavePlayer_26
   WavePlayer_25
   WavePlayer_27
   WavePlayer_28
   WavePlayer_8
   WavePlayer_30
   WavePlayer_31
   WavePlayer_29
   null
  ]
  properties {inputVolume [1,1,1,1,1,1,1,1,1,1,1]}
 }
 {
  id=Concatenator_3
  children [
   WavePlayer_33
   WavePlayer_32
   WavePlayer_37
   WavePlayer_35
   WavePlayer_36
   WavePlayer_38
   WavePlayer_41
   WavePlayer_34
   WavePlayer_39
   WavePlayer_40
   null
  ]
  properties {inputVolume [1,1,1,1,1,1,1,1,1,1,1]}
 }
 {
  id=Concatenator_4
  children [
   WavePlayer_51
   WavePlayer_45
   WavePlayer_47
   WavePlayer_42
   WavePlayer_49
   WavePlayer_43
   WavePlayer_46
   WavePlayer_44
   WavePlayer_48
   WavePlayer_50
   null
  ]
  properties {inputVolume [1,1,1,1,1,1,1,1,1,1,1]}
 }
 {id=Random_1, children [Concatenator_2, Concatenator_3, Concatenator_4, Concatenator_1], properties {weights [1,1,1,1]}}
 {id=WavePlayer_24, properties {soundWaveAssetPtr=&/Phone_Prank_Dial_03}}
 {id=WavePlayer_25, properties {soundWaveAssetPtr=&/Phone_Prank_Dial_04}}
 {id=WavePlayer_26, properties {soundWaveAssetPtr=&/Phone_Prank_Dial_05}}
 {id=WavePlayer_27, properties {soundWaveAssetPtr=&/Phone_Prank_Dial_06}}
 {id=WavePlayer_28, properties {soundWaveAssetPtr=&/Phone_Prank_Dial_07}}
 {id=WavePlayer_29, properties {soundWaveAssetPtr=&/Phone_Prank_Dial_08}}
 {id=WavePlayer_30, properties {soundWaveAssetPtr=&/Phone_Prank_Dial_09}}
 {id=WavePlayer_31, properties {soundWaveAssetPtr=&/Phone_Prank_Dial_10}}
 {id=WavePlayer_32, properties {soundWaveAssetPtr=&/Phone_Prank_Dial_01}}
 {id=WavePlayer_33, properties {soundWaveAssetPtr=&/Phone_Prank_Dial_02}}
 {id=WavePlayer_34, properties {soundWaveAssetPtr=&Phone_Prank_Dial_03}}
 {id=WavePlayer_35, properties {soundWaveAssetPtr=&Phone_Prank_Dial_04}}
 {id=WavePlayer_36, properties {soundWaveAssetPtr=&Phone_Prank_Dial_05}}
 {id=WavePlayer_37, properties {soundWaveAssetPtr=&Phone_Prank_Dial_06}}
 {id=WavePlayer_38, properties {soundWaveAssetPtr=&Phone_Prank_Dial_07}}
 {id=WavePlayer_39, properties {soundWaveAssetPtr=&Phone_Prank_Dial_08}}
 {id=WavePlayer_40, properties {soundWaveAssetPtr=&Phone_Prank_Dial_09}}
 {id=WavePlayer_41, properties {soundWaveAssetPtr=&Phone_Prank_Dial_10}}
 {id=WavePlayer_42, properties {soundWaveAssetPtr=&Phone_Prank_Dial_01}}
 {id=WavePlayer_43, properties {soundWaveAssetPtr=&Phone_Prank_Dial_02}}
 {id=WavePlayer_44, properties {soundWaveAssetPtr=&Phone_Prank_Dial_03}}
 {id=WavePlayer_45, properties {soundWaveAssetPtr=&Phone_Prank_Dial_04}}
 {id=WavePlayer_46, properties {soundWaveAssetPtr=&Phone_Prank_Dial_05}}
 {id=WavePlayer_47, properties {soundWaveAssetPtr=&Phone_Prank_Dial_06}}
 {id=WavePlayer_48, properties {soundWaveAssetPtr=&Phone_Prank_Dial_07}}
 {id=WavePlayer_49, properties {soundWaveAssetPtr=&Phone_Prank_Dial_08}}
 {id=WavePlayer_50, properties {soundWaveAssetPtr=&Phone_Prank_Dial_09}}
 {id=WavePlayer_51, properties {soundWaveAssetPtr=&Phone_Prank_Dial_10}}
 {id=WavePlayer_52, properties {soundWaveAssetPtr=&Phone_Prank_Dial_01}}
 {id=WavePlayer_53, properties {soundWaveAssetPtr=&Phone_Prank_Dial_02}}
 {id=WavePlayer_54, properties {soundWaveAssetPtr=&Phone_Prank_Dial_03}}
 {id=WavePlayer_55, properties {soundWaveAssetPtr=&Phone_Prank_Dial_04}}
 {id=WavePlayer_56, properties {soundWaveAssetPtr=&Phone_Prank_Dial_05}}
 {id=WavePlayer_57, properties {soundWaveAssetPtr=&Phone_Prank_Dial_06}}
 {id=WavePlayer_58, properties {soundWaveAssetPtr=&Phone_Prank_Dial_07}}
 {id=WavePlayer_59, properties {soundWaveAssetPtr=&Phone_Prank_Dial_08}}
 {id=WavePlayer_60, properties {soundWaveAssetPtr=&Phone_Prank_Dial_09}}
 {id=WavePlayer_61, properties {soundWaveAssetPtr=&Phone_Prank_Dial_10}}
 {id=WavePlayer_7, properties {soundWaveAssetPtr=&Phone_Prank_Dial_01}}
 {id=WavePlayer_8, properties {soundWaveAssetPtr=&Phone_Prank_Dial_02}}
]
properties {bOverrideAttenuation=true, maxDistance=1400, pitchMultiplier=1.021, volumeMultiplier=1.5029, attenuationOverrides {falloffDistance=1000}}
