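# Sound cue for the BerryCard ring sound in WildEstate gameplay, built from a phone prank call wave.
# A single wave player feeds the root directly, with no randomization or mixing, and the cue boosts its volume.
# It references the Fortnite payphone ambient attenuation asset and also stores logarithmic falloff values in its attenuation overrides.
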SoundCue /WildEstate/Gameplay/BerryCard/Audio/SC_BerryCard_Ring_01
root=WavePlayer_0
nodes [{id=WavePlayer_0, properties {soundWaveAssetPtr=&/Phone_Prank_Call_04}}]
properties {
 attenuationSettings=/Game/Sounds/Attenuation/Fort_Payphone_Amb_Att
 maxDistance=1700
 volumeMultiplier=1.4983
 attenuationOverrides {distanceAlgorithm=Logarithmic, falloffDistance=4719.0283, attenuationShapeExtents {x=407.5906}}
}
