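# Sound cue for the second Ring variant in the BerryCard gameplay audio, playing a phone prank call wave.
# One wave player feeds two parallel attenuation nodes that borrow Fort payphone ambience settings: a cone and a sphere whose name indicates occlusion.
# An evenly weighted mixer blends both spatial layers of the same sound, with the cue's overall volume boosted.
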SoundCue /WildEstate/Gameplay/BerryCard/Audio/SC_BerryCard_Ring_02
root=Mixer_0
nodes [
 {id=Attenuation_0, children [WavePlayer_5], properties {attenuationSettings=/Game/Sounds/Attenuation/Fort_Payphone_Amb_Cone_Att}}
 {id=Attenuation_1, children [WavePlayer_5], properties {attenuationSettings=/Game/Sounds/Attenuation/Fort_Payphone_Amb_Sphere_Occ_Att}}
 {id=Mixer_0, children [Attenuation_0, Attenuation_1], properties {inputVolume [0.5,0.5]}}
 {id=WavePlayer_5, properties {soundWaveAssetPtr=&/Phone_Prank_Call_04}}
]
properties {
 maxDistance=1700
 volumeMultiplier=1.4983
 attenuationOverrides {distanceAlgorithm=Logarithmic, falloffDistance=4719.0283, attenuationShapeExtents {x=407.5906}}
}
