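# Sound cue for the Seymour Butz NPC voice line tied to berry card pickup in the WildEstate gameplay feature.
# A single wave player feeds the root directly, with no randomization, mixing or modulation nodes.
# The cue routes through the NPC voice sound class with boosted volume and a berry VO attenuation asset.
# It also carries attenuation override values for falloff distance and distance-based reverb send.
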SoundCue /WildEstate/Gameplay/BerryCard/Audio/VO_IP_BerryCard_SeymourButz_Pickup_Cue
root=WavePlayer_5
nodes [{id=WavePlayer_5, properties {soundWaveAssetPtr=&/VO_IP_BerryCard_SeymourButz}}]
properties {
 attenuationSettings=&/ATT_BR_Fort_Berry_VO
 maxDistance=2300
 soundClassObject=&Fort_VO_NPC
 volumeMultiplier=1.6
 attenuationOverrides {falloffDistance=2000, reverbDistanceMax=5560.9712, reverbDistanceMin=580.4022, reverbWetLevelMax=1.2677, reverbWetLevelMin=0.7373}
}
