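; Base interactable phone-style prop (BerryCard) that rings for nearby players, plays voice-over when answered and drops loot once.
; Server-side sphere overlaps and a re-check timer trigger multicast ring audio, indicator and effects, gated by hotfixable cooldowns and a HasUsed tag.
; Interacting after a recent ring fires a delayed voice-over GameplayCue, then applies HasUsed to pawns in range; otherwise it plays a dial tone.
; Busy state blocks interaction during playback and after death.
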(event EventBeginPlay
  (bind _ringsound (Variables|Ring|GetRingSound))
  ;[0 1 0] BeginPlay
  (|CallParentFunction)
  (Utilities|FlowControl|Sequence
    (:then_0
      (if (Variables|Default|GetEnabled)
        (CallFunction|EnableBerryCard)
        (else
          (CallFunction|DisableBerryCard))))
    (:then_1
      (Audio|Components|Audio|SetSound (Variables|Ring|GetRingAudio) _ringsound)
      (Variables|/Script/FortniteGame.FortSoundIndicatorComponent|SetDefaultSoundLinkedToIndicator _ringsound (Variables|Ring|GetFortSoundIndicator))
      (Variables|Actor|SetbCanBeDamaged (ScalableFloat|GetValueasBool (Variables|Hotfix|GetHF_CanBeDamaged))))))

(event Custom|StartAlreadyOverlappedTimer
  ;[1 1 0] Start a timer to see if we want to ring for players that are in the area. Should replace this with a check to see if there are any pawns in the area at all.
  (Utilities|Time|ClearandInvalidateTimerbyHandle (Variables|State|GetAlreadyOverlappedTimer))
  (bind _returnvalue (Utilities|Time|SetTimerbyEvent (EventDispatchers|CreateEvent :!event Custom|RingForAlreadyOverlappedPawns) 1.0 true))
  (Variables|State|SetAlreadyOverlappedTimer _returnvalue))

(event Custom|ApplyHasHeardGameplayEffectToActorsInRange
  ;[1 1 0] We probably want better overlap task and logic to determine if a player has heard the VO
  (bind _outactors (Collision|SphereOverlapActors (Transformation|GetActorLocation) 1.25e+03 (Utilities|Array|MakeArray "ObjectTypeQuery3")))
  (for _array_element _outactors
    (bind _asfort_player_pawn_athena (Utilities|Casting|CastToFortPlayerPawnAthena _array_element))
    (Fortnite|ApplyGameplayEffectsToActor _asfort_player_pawn_athena (Utilities|Array|MakeArray "/WildEstate/Gameplay/BerryCard/GE_BerryCard_HasUsed.GE_BerryCard_HasUsed_C") (Transformation|GetActorLocation) false 0 (Ability|GetAbilitySystemComponent 0) (Class|AthenaFunctionLibrary|MakeSimpleFortEffectContext self self))))

(event Interaction|EventBlueprintOnInteract
  (params (InteractingPawn :type Fort%20Pawn%20Object%20Reference) (InteractionBeingAttempted :type EInteractionBeingAttempted%20Enum :default "FirstInteraction"))
  ;[0 1 0.204] OnInteract Server
  ; Play VO if the interacting pawn has not done so already.
  (|CallParentFunction InteractingPawn InteractionBeingAttempted)
  (CallFunction|TrySpawnLoot)
  (if (CallFunction|ShouldPlayVO)
    (CallFunction|PlayVO)
    (else
      (CallFunction|PlayDialtoneSound))))

(event Custom|DisableInteractForDuration
  (params (Duration :type Float%20%28double-precision%29))
  ;[1 1 0] Disable interact while in use
  (Utilities|Time|SetTimerbyEvent (EventDispatchers|CreateEvent :!event Custom|EnableInteract) Duration)
  (Variables|Interaction|SetbAllowInteract false)
  (Variables|State|SetIsBusy true))

(event Custom|EnableInteract
  ;[1 1 0] Disable interact while in use
  (Variables|Interaction|SetbAllowInteract true)
  (Variables|State|SetIsBusy false))

(event OnComponentBeginOverlap(SphereCollision)
  (params (OverlappedComponent :type Primitive%20Component%20Object%20Reference) (OtherActor :type Actor%20Object%20Reference) (OtherComp :type Primitive%20Component%20Object%20Reference) (OtherBodyIndex :type Integer) (bFromSweep :type Boolean) (SweepResult :type Hit%20Result%20Structure%20%28by%20ref%29))
  ;[0.1 0.1042 1] Player overlapped - Check to play ringing sound
  (Networking|SwitchHasAuthority
    (:Authority
      (bind _asfort_player_pawn_athena (Utilities|Casting|CastToFortPlayerPawnAthena OtherActor))
      (if (and (not (GameplayTags|HasMatchingGameplayTag _asfort_player_pawn_athena (Variables|State|GetHasUsedTag))) (CallFunction|ShouldPlayRing OtherActor))
        (CallFunction|PlayRingingSound)))))

(event Custom|EnableBerryCard
  ;[1 0 0.8694] Enabling and Disabling the ringing and voice overs
  (Collision|SetCollisionEnabled (Variables|Default|GetSphereCollision) "QueryOnly")
  (Networking|SwitchHasAuthority
    (:Authority
      (CallFunction|StartAlreadyOverlappedTimer)
      (CallFunction|UpdateVoiceOverAudio))))

(event Custom|DisableBerryCard
  ;[1 0 0.8694] Enabling and Disabling the ringing and voice overs
  (Collision|SetCollisionEnabled (Variables|Default|GetSphereCollision) "NoCollision")
  (Networking|SwitchHasAuthority
    (:Authority
      (Utilities|Time|ClearandInvalidateTimerbyHandle (Variables|State|GetAlreadyOverlappedTimer)))))

(event Custom|RingForAlreadyOverlappedPawns
  ;[1 0.03907 0.7358] Ring for already overlapped players
  (Networking|SwitchHasAuthority
    (:Authority
      (for _array_element (Collision|GetOverlappingActors (Variables|Default|GetSphereCollision) 0 :!class /Script/Engine.PrimitiveComponent)
        (if (CallFunction|ShouldPlayRing _array_element)
          (CallFunction|PlayRingingSound))))))

(event Custom|PlayVO
  ;[0.1 0.1042 1] Play Audio
  (if (not (Variables|State|GetIsBusy))
    (Variables|State|SetCanPlayVO false)
    (CallFunction|DisableInteractforDuration (ScalableFloat|GetValueatLevel (Variables|Hotfix|GetHF_OnVOInteractDisableDuration)))
    (CallFunction|MULTIStopRing)
    (Utilities|Time|ClearandInvalidateTimerbyHandle (Variables|Ring|GetForcePlayRingingTimer))
    (CallFunction|UpdateVoiceOverAudio)
    (Utilities|FlowControl|Delay (CallFunction|GetAudioDelayDuration))
    (GameplayCue|ExecuteGameplayCue self (Variables|GameplayCues|GetVoiceOverGameplayCue) (Class|AthenaFunctionLibrary|MakeSimpleFortEffectContext self self))
    (Variables|State|SetLastVOTime (Utilities|Time|GetGameTimeinSeconds))
    (CallFunction|ApplyHasHeardGameplayEffecttoActorsinRange)))

(event Custom|PlayRingingSound
  ;[0.1 0.1042 1] Play Audio
  (if (not (Variables|State|GetIsBusy))
    (Variables|State|SetCanPlayVO true)
    (Variables|State|SetLastRingTime (Utilities|Time|GetGameTimeinSeconds))
    (Utilities|Time|ClearandInvalidateTimerbyHandle (Variables|State|GetAlreadyOverlappedTimer))
    (CallFunction|MULTIPlayRing)
    (Utilities|Time|SetTimerbyEvent (EventDispatchers|CreateEvent :!event Custom|StartAlreadyOverlappedTimer) (ScalableFloat|GetValueatLevel (Variables|Hotfix|GetHF_TimeToRestartAlreadyOverlappedCheck)))
    (bind _returnvalue (Utilities|Time|SetTimerbyEvent (EventDispatchers|CreateEvent :!event Custom|ForcePlayRingingSound) 4.5))
    (Variables|Ring|SetForcePlayRingingTimer _returnvalue)))

(event Custom|ForcePlayRingingSound
  ;[0.1 0.1042 1] Play Audio
  (if (not (Variables|State|GetIsBusy))
    (CallFunction|MULTIPlayRing)))

(event Custom|PlayDialtoneSound
  ;[0.1 0.1042 1] Play Audio
  (if (not (Variables|State|GetIsBusy))
    (CallFunction|DisableInteractforDuration (ScalableFloat|GetValueatLevel (Variables|Hotfix|GetHF_OnDialToneInteractDisableDuration)))
    (Utilities|Time|ClearandInvalidateTimerbyHandle (Variables|Ring|GetForcePlayRingingTimer))
    (CallFunction|MULTIStopRing)
    (Utilities|FlowControl|Delay (CallFunction|GetAudioDelayDuration))
    (GameplayCue|ExecuteGameplayCue self (Variables|GameplayCues|GetDialtoneGameplayCue) (Class|AthenaFunctionLibrary|MakeSimpleFortEffectContext self self))))

(event Custom|SetVoiceOverSound
  (params (NewSound :type Sound%20Base%20Object%20Reference))
  ;[0.1 0.1042 1] Play Audio
  (Networking|SwitchHasAuthority
    (:Authority
      (Variables|VO|SetVoiceOverSoundCurrent NewSound))))

(event Custom|MULTI_PlayRing
  ;[1 0.8693 0] Multicast rpc to have everyone play the ring sound
  (Audio|Components|Audio|Play (Variables|Ring|GetRingAudio) 0.0)
  (SoundIndicator|ShowSoundIndicator (Variables|Ring|GetFortSoundIndicator) (Transformation|GetActorLocation) 0 1.0 0)
  (Components|Activation|Activate (Variables|Default|GetNiagara) true)
  (Rendering|Material|SetScalarParameterforCustomPrimitiveData (Variables|Debug|GetStaticMeshComponent) "Rotational WPO BP Time" (Utilities|Time|GetGameTimeinSeconds)))

(event Custom|MULTI_StopRing
  ;[1 0.8693 0] Multicast rpc to have everyone play the ring sound
  (Audio|Components|Audio|Stop (Variables|Ring|GetRingAudio))
  (Components|Activation|Deactivate (Variables|Default|GetNiagara)))

(event Custom|TrySpawnLoot
  ;[0.1 0.1042 1] SpawnLoot
  (if (not (Variables|State|GetHasSpawnedLoot))
    (Variables|State|SetHasSpawnedLoot true)
    (Class|AthenaFunctionLibrary|SpawnLootfromTable (Variables|Loot|GetLootTierGroup) 1 -1 (+ (Utilities|Operators|Add (Transformation|GetActorLocation) (Math|Vector|MakeVector 0.0 0.0 (Variables|Loot|GetLootOriginZOffset))) (* (Transformation|GetForwardVector (Variables|Default|GetArrow)) (Variables|Loot|GetLootOriginForwardOffset))) (Transformation|GetForwardVector (Variables|Default|GetArrow)) 0 true false false 0 "Container" "ItemSpawner")
    (CallFunction|PlayLootFX)))

(event Custom|PlayLootFX
  ;[0.1 0.1042 1] SpawnLoot
  (Audio|PlaySoundatLocation (Variables|Loot|GetLootSound) (Transformation|GetActorLocation)))

(event EventOnDeathPlayEffects
  (params (Damage :type Float%20%28single-precision%29) (DamageTags :type Gameplay%20Tag%20Container%20Structure%20%28by%20ref%29) (Momentum :type Vector :default "0, 0, 0") (HitInfo :type Hit%20Result%20Structure%20%28by%20ref%29) (InstigatedBy :type Fort%20Pawn%20Object%20Reference) (DamageCauser :type Actor%20Object%20Reference) (EffectContext :type Gameplay%20Effect%20Context%20Handle%20Structure))
  ;[1 0.3 0 0.5] Stop audio on death
  (|CallParentFunction Damage DamageTags Momentum HitInfo InstigatedBy DamageCauser EffectContext)
  (Audio|Components|Audio|Stop (Variables|Ring|GetRingAudio)))

(event EventOnDeathServer
  (params (Damage :type Float%20%28single-precision%29) (DamageTags :type Gameplay%20Tag%20Container%20Structure%20%28by%20ref%29) (Momentum :type Vector :default "0, 0, 0") (HitInfo :type Hit%20Result%20Structure%20%28by%20ref%29) (InstigatedBy :type Controller%20Object%20Reference) (DamageCauser :type Actor%20Object%20Reference) (EffectContext :type Gameplay%20Effect%20Context%20Handle%20Structure))
  ;[1 0.3 0 0.5] Set busy so nothing will execute on death
  (|CallParentFunction Damage DamageTags Momentum HitInfo InstigatedBy DamageCauser EffectContext)
  (Variables|State|SetIsBusy true))

(fn BlueprintCanInteract
  (params (InteractingPawn :type Fort%20Pawn%20Object%20Reference) (InteractionBeingAttempted :type EInteractionBeingAttempted%20Enum :default "FirstInteraction") (InteractionType :type TInteractionType%20Enum :default "IT_NoInteraction"))
  (return (and (Variables|Interaction|GetbAllowInteract) (not (Variables|State|GetIsBusy)))))

(fn BlueprintGetInteractionString
  (params (InteractingPawn :type Fort%20Pawn%20Object%20Reference) (InteractionBeingAttempted :type EInteractionBeingAttempted%20Enum :default "FirstInteraction"))
  (return (Variables|GetBuildingActorInteractionText)))

(fn BlueprintOnLocalInteract
  (params (InteractingPawn :type Fort%20Player%20Pawn%20Object%20Reference) (InteractionBeingAttempted :type EInteractionBeingAttempted%20Enum :default "FirstInteraction"))
  (Utilities|FlowControl|Sequence
    (:then_0
      (if (GameplayTags|HasMatchingGameplayTag InteractingPawn (Variables|State|GetHasUsedTag))
        (CallFunction|PlayDialtoneSound)
        (return false)
        (else
          _)))
    (:then_1
      _))
  (bind _returnvalue (|CallParentFunction 0))
  (return _returnvalue))

(fn GetAudioDelayDuration
  (return (Math|Float|Clamp(Float) (ScalableFloat|GetValueatLevel (Variables|Hotfix|GetHF_OnInteractAudioDelay)) 0.1 3.0)))

(fn GetDisableInteractDuration
  (return (Math|Float|Clamp(Float) (ScalableFloat|GetValueatLevel (Variables|Hotfix|GetHF_OnInteractAudioDelay)) 0.1 2.0)))

(fn LocalOnFailedInteract
  (params (InteractingPawn :type Fort%20Player%20Pawn%20Object%20Reference))
  (|CallParentFunction InteractingPawn))

(fn ShouldPlayRing
  (params (OverlappingActor :type Actor%20Object%20Reference))
  (bind _returnvalue (GameplayTags|HasMatchingGameplayTag(Message) OverlappingActor (Variables|State|GetHasUsedTag)))
  (return (Math|Boolean|ANDBoolean (not _returnvalue) (> (- (Utilities|Time|GetGameTimeinSeconds) (Variables|State|GetLastVOTime)) (ScalableFloat|GetValueatLevel (Variables|Hotfix|GetHF_TimeBetweenVOMin))) (or (Math|Float|NearlyEqual(Float) (Variables|State|GetLastRingTime)) (> (- (Utilities|Time|GetGameTimeinSeconds) (Variables|State|GetLastRingTime)) (ScalableFloat|GetValueatLevel (Variables|Hotfix|GetHF_TimeBetweenRingsMin)))))))

(fn ShouldPlayVO
  (return (Math|Boolean|ANDBoolean (< (- (Utilities|Time|GetGameTimeinSeconds) (Variables|State|GetLastRingTime)) (ScalableFloat|GetValueatLevel (Variables|Hotfix|GetHF_TimeSinceLastRingToHearVO))) (not (Math|Float|NearlyEqual(Float) (Variables|State|GetLastRingTime) 0.0 0.01)) (Variables|State|GetCanPlayVO))))

(fn UpdateVoiceOverAudio
  (bind _voiceoversoundsrandom (Variables|VO|GetVoiceOverSoundsRandom))
  (bind _voiceoversoundoverride (Variables|VO|GetVoiceOverSoundOverride)
    (:then
      (CallFunction|SetVoiceOverSound _voiceoversoundoverride))
    (:else
      (Utilities|Array|Shuffle _voiceoversoundsrandom)
      (if (Utilities|Array|IsValidIndex _voiceoversoundsrandom)
        (CallFunction|SetVoiceOverSound (Utilities|Array|Get(acopy) _voiceoversoundsrandom))))))

(fn ConstructionScript)
