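; Blueprint for the payphone berry card in WildEstate, apparently a child variant that inherits its behavior from a parent class.
; The multicast-named PlayRing event and the construction script only call their parent implementations.
; The file adds no logic of its own, so any ring behavior comes from the parent.
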(event EventMULTI_PlayRing
  (|CallParentFunction))

(fn ConstructionScript
  (|CallParentFunction))
